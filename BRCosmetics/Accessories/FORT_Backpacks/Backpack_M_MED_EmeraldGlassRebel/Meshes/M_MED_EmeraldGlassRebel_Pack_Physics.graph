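# Physics asset for the EmeraldGlassRebel backpack cosmetic, defining collision and joint limits for the pack and its dangling belt and ring.
# Simple box, capsule and sphere bodies cover root, pack, ring, belt and belt end, with a large root box.
# Belt end chains to belt, and belt and ring each attach to the pack with angular limits, tightest on the ring's swing.
# The pack-to-root joint leaves swing and twist free.
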PhysicsAsset /BRCosmetics/Accessories/FORT_Backpacks/Backpack_M_MED_EmeraldGlassRebel/Meshes/M_MED_EmeraldGlassRebel_Pack_Physics
bodies {
 dyn_belt [{shapeType=Box, extentX=7.5, extentY=5.5, extentZ=1, shapeName=dyn_belt_box, center=<6.2823 0 0>}]
 dyn_belt_end [{shapeType=Capsule, length=6, radius=1.75, shapeName=dyn_belt_end_capsule, center=<4.1584 0 0>, rotation=<rot 90 0 -0.0002>}]
 dyn_ring [{shapeType=Box, extentX=11, extentY=10.5, extentZ=2, shapeName=dyn_ring_box, center=<0 5.1773 0>, rotation=<rot -0.0001 0 -3.055>}]
 pack [
  {shapeType=Sphere, radius=6.25, shapeName=pack_sphere, center=<-3.01 -18.2681 4.9302>}
  {
   shapeType=Capsule
   length=4.9997
   radius=9.2383
   shapeName=pack_capsule
   center=<0.3044 -9.81 -3.9585>
   rotation=<rot -0.0001 -23.6108 -16.2149>
  }
 ]
 root [{shapeType=Box, extentX=34.5, extentY=5.5, extentZ=42, shapeName=root_box, center=<0 2.6821 -10.7051>, rotation=<rot 0 0 -0.6423>}]
}
constraints [
 {bone1Name=dyn_belt, bone2Name=pack, swing1LimitDegrees=10, swing2LimitDegrees=45, twistLimitDegrees=30}
 {bone1Name=dyn_belt_end, bone2Name=dyn_belt, swing1LimitDegrees=15, swing2LimitDegrees=90, twistLimitDegrees=45}
 {
  bone1Name=pack
  bone2Name=root
  swing1LimitDegrees=45
  swing1Motion=Free
  swing2LimitDegrees=45
  swing2Motion=Free
  twistLimitDegrees=45
  twistMotion=Free
 }
 {bone1Name=dyn_ring, bone2Name=pack, swing1LimitDegrees=2, swing2LimitDegrees=2, twistLimitDegrees=27}
]
